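; Figment Chapter 1 Moon time-of-day manager Blueprint with hooks for sky, cloud and lighting material state.
; Custom events write cloud time and storm phase to the safe-zone clouds parameter collection.
; SetMPCValues copies real time, sun color and axes, and fog colors into global material parameters.
; A volumetric cloud function checks the quality CVar, then blocking-loads the material or hides the clouds, toggling SkyBoxClouds; several other hooks are empty.
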(event EventSetVolumetricCloudMaterial)

(event EventEventSetVolCloudMaterial)

(event EventOnTimeOfDayManagerVisibilitySet
  (params (bVisible :type Boolean))
  (|CallParentFunction bVisible))

(event Custom|SetCloudsCustomTime
  (params (customTime :type Float%20%28double-precision%29))
  (Rendering|Material|SetScalarParameterValue "/Game/Athena/SafeZone/MPC_SafeZoneClouds.MPC_SafeZoneClouds" "CustomCloudsTime" customTime :!class /Script/Engine.KismetMaterialLibrary))

(event Custom|SetStormPhase
  (params (StormPhase :type Float%20%28double-precision%29))
  (Rendering|Material|SetScalarParameterValue "/Game/Athena/SafeZone/MPC_SafeZoneClouds.MPC_SafeZoneClouds" "StormPhase" StormPhase :!class /Script/Engine.KismetMaterialLibrary))

(event Custom|SetStormPhaseCount
  (params (StormPhaseCount :type Integer :default 5))
  (Rendering|Material|SetScalarParameterValue "/Game/Athena/SafeZone/MPC_SafeZoneClouds.MPC_SafeZoneClouds" "StormPhaseCount" (Math|Conversions|ToFloat(Integer) StormPhaseCount) :!class /Script/Engine.KismetMaterialLibrary))

(event Custom|SetSkyLightCubeMapBlend_1
  (params (val :type Float%20%28double-precision%29)))

(event Sequence|EventUpdateScalabilitySequences
  (CallFunction|UpdateScalability))

(fn Set2DCloudMaterial_Figment)

(fn SetMPCValues
  (Rendering|Material|SetScalarParameterValue "/Game/Packages/Fortress_SharedMaterials/GlobalMaterialParameters/FortniteMaterialParameters.FortniteMaterialParameters" "RealGameTime" (Utilities|Time|GetRealTimeSeconds) :!class /Script/Engine.KismetMaterialLibrary)
  (Rendering|Material|SetVectorParameterValue "/Game/Packages/Fortress_SharedMaterials/GlobalMaterialParameters/FortniteMaterialParameters.FortniteMaterialParameters" "SunLightColor" (Math|Conversions|ToLinearColor(Color) (Variables|/Script/Engine.LightComponentBase|GetLightColor (Variables|DayNightCycle|GetSunComponent))) :!class /Script/Engine.KismetMaterialLibrary)
  (Rendering|Material|SetVectorParameterValue "/Game/Packages/Fortress_SharedMaterials/GlobalMaterialParameters/FortniteMaterialParameters.FortniteMaterialParameters" "FogColor" (Variables|/Script/Engine.ExponentialHeightFogComponent|GetFogInscatteringLuminance (Variables|DayNightCycle|GetExponentialHeightFogComponent)) :!class /Script/Engine.KismetMaterialLibrary)
  (Rendering|Material|SetVectorParameterValue "/Game/Packages/Fortress_SharedMaterials/GlobalMaterialParameters/FortniteMaterialParameters.FortniteMaterialParameters" "FogDirectionalInscatteringColor" (Variables|/Script/Engine.ExponentialHeightFogComponent|GetDirectionalInscatteringLuminance (Variables|DayNightCycle|GetExponentialHeightFogComponent)) :!class /Script/Engine.KismetMaterialLibrary)
  (Rendering|Material|SetVectorParameterValue "/Game/Packages/Fortress_SharedMaterials/GlobalMaterialParameters/FortniteMaterialParameters.FortniteMaterialParameters" "SunlightYTransformVector" (Math|Conversions|ToLinearColor(Vector) (Transformation|GetRightVector (Variables|DayNightCycle|GetSunComponent))) :!class /Script/Engine.KismetMaterialLibrary)
  (Rendering|Material|SetVectorParameterValue "/Game/Packages/Fortress_SharedMaterials/GlobalMaterialParameters/FortniteMaterialParameters.FortniteMaterialParameters" "SunlightZTransformVector" (Math|Conversions|ToLinearColor(Vector) (Transformation|GetUpVector (Variables|DayNightCycle|GetSunComponent))) :!class /Script/Engine.KismetMaterialLibrary)
  (Rendering|Material|SetVectorParameterValue "/Game/Packages/Fortress_SharedMaterials/GlobalMaterialParameters/FortniteMaterialParameters.FortniteMaterialParameters" "SunlightDirectionalVector" (Math|Conversions|ToLinearColor(Vector) (* (Transformation|GetForwardVector (Variables|DayNightCycle|GetSunComponent)) -1.0)) :!class /Script/Engine.KismetMaterialLibrary))

(fn SetVolumetricCloudMaterial_Figment
  (bind _returnvalue (Development|GetConsoleVariableIntValue "Athena.VolumetricClouds.Quality"))
  (if (> _returnvalue 0)
    (bind _returnvalue_2 (Utilities|LoadAssetBlocking (Variables|Clouds|GetVolumetricCloudMaterial)))
    (bind _asmaterial_interface (Utilities|Casting|CastToMaterialInterface _returnvalue_2))
    (Rendering|SetMaterial (Variables|DayNightCycle|GetVolumetricCloudComponent) _asmaterial_interface)
    (Rendering|Material|SetScalarParameterValue "/Game/Packages/Fortress_SharedMaterials/GlobalMaterialParameters/FortniteMaterialParameters.FortniteMaterialParameters" "SkyBoxClouds" :!class /Script/Engine.KismetMaterialLibrary)
    (Fortnite|FortShippingLog self "TODM Scalability: Volumetric Clouds = True")
    (else
      (Rendering|SetMaterial (Variables|DayNightCycle|GetVolumetricCloudComponent) 0)
      (Rendering|SetVisibility (Variables|DayNightCycle|GetVolumetricCloudComponent) false false)
      (Rendering|Material|SetScalarParameterValue "/Game/Packages/Fortress_SharedMaterials/GlobalMaterialParameters/FortniteMaterialParameters.FortniteMaterialParameters" "SkyBoxClouds" 1.0 :!class /Script/Engine.KismetMaterialLibrary)
      (Fortnite|FortShippingLog self "TODM Scalability: Volumetric Clouds = False"))))

(fn UpdateScalability
  (|CallParentFunction))

(fn ConstructionScript
  (|CallParentFunction))

(fn ValidateVolumeFog)
